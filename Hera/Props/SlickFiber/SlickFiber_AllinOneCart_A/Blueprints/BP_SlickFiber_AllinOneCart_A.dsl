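; Fortnite building prop Blueprint for the SlickFiber all-in-one cart A, which tints the prop's accessories mesh.
; The construction script runs the parent's, then reads the building texture data array and passes an entry's tint color to SetColoronSecondaryMesh for that mesh.
; If the array index check fails, it passes solid red instead.
; BeginPlay, overlap and Tick events are empty.
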(event EventBeginPlay)

(event Collision|EventActorBeginOverlap
  (params (OtherActor :type Actor%20Object%20Reference)))

(event EventTick
  (params (DeltaSeconds :type Float%20%28single-precision%29)))

(fn ConstructionScript
  (bind _sm_slickfiber_accesories_a (Variables|Default|GetSM_SlickFiber_Accesories_A))
  (|CallParentFunction)
  (bind _outtexturedataarray (Building|GetBuildingTextureData))
  (if (Utilities|Array|IsValidIndex _outtexturedataarray)
    (Default|SetColoronSecondaryMesh (Math|Conversions|ToLinearColor(Color) (Variables|/Script/FortniteGame.BuildingTextureData|GetTintColor (Utilities|Array|Get(acopy) _outtexturedataarray (LocalVariables|GetTDIndex)))) _sm_slickfiber_accesories_a)
    (else
      (Default|SetColoronSecondaryMesh "(R=1.000000,G=0.000000,B=0.000000,A=1.000000)" _sm_slickfiber_accesories_a))))
